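# Physics asset for the F_MED_Musician backpack cosmetic, defining secondary motion for its dangling keychain, handle and buttons.
# A single capsule on the pack bone anchors the handle sphere, three small button capsules and the first keychain link.
# The keychain forms a three-link capsule chain with wider swing further along, while twist stays tightly limited throughout.
# Buttons and handle swing mostly along one axis, keeping their movement subtle and hinge-like.
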PhysicsAsset /BRCosmetics/Accessories/FORT_Backpacks/F_MED_Musician/Meshes/F_MED_Musician_Pack_Physics
bodies {
 dyn_button_01 [{shapeType=Capsule, length=0.505, radius=0.505, center=<0.2447 -0.0274 -0.0445>, rotation=<rot 1.7817 -90 -10.5076>}]
 dyn_button_02 [{shapeType=Capsule, length=1.2388, radius=0.5, center=<0.278 0.0161 -0.0489>, rotation=<rot 78.8229 -174.3929 -84.285>}]
 dyn_button_03 [{shapeType=Capsule, length=0.505, radius=0.505, center=<0.2419 -0.024 -0.0333>, rotation=<rot 2.0391 -90 -13.3838>}]
 dyn_handle [{shapeType=Sphere, radius=2.7947, center=<1.3398 -0.004 -0.013>}]
 dyn_keychain_01 [{shapeType=Capsule, length=0.255, radius=1.755, center=<2.7649 0 0>, rotation=<rot 90 0 0>}]
 dyn_keychain_02 [{shapeType=Capsule, length=0.505, radius=1.755, center=<2.0801 -0.0001 0>, rotation=<rot 90 0 0>}]
 dyn_keychain_03 [{shapeType=Capsule, length=0.505, radius=1.755, center=<1.7074 -0.0001 0>, rotation=<rot 90 0 0>}]
 pack [{shapeType=Capsule, length=7.8977, radius=6.5422, center=<-9.7843 -4.6895 -13.5797>, rotation=<rot -0.7252 -89.9949 -48.4673>}]
}
constraints [
 {bone1Name=dyn_keychain_01, bone2Name=pack, swing1LimitDegrees=2, swing2LimitDegrees=30, twistLimitDegrees=1}
 {bone1Name=dyn_keychain_02, bone2Name=dyn_keychain_01, swing1LimitDegrees=45, swing2LimitDegrees=45, twistLimitDegrees=1}
 {bone1Name=dyn_keychain_03, bone2Name=dyn_keychain_02, swing1LimitDegrees=12, swing2LimitDegrees=45, twistLimitDegrees=1}
 {bone1Name=dyn_handle, bone2Name=pack, swing1LimitDegrees=36, swing2LimitDegrees=1, twistLimitDegrees=1}
 {bone1Name=dyn_button_01, bone2Name=pack, swing1LimitDegrees=1, swing2LimitDegrees=8, twistLimitDegrees=4}
 {bone1Name=dyn_button_02, bone2Name=pack, swing1LimitDegrees=1, swing2LimitDegrees=8, twistLimitDegrees=4}
 {bone1Name=dyn_button_03, bone2Name=pack, swing1LimitDegrees=1, swing2LimitDegrees=12, twistLimitDegrees=4}
]
